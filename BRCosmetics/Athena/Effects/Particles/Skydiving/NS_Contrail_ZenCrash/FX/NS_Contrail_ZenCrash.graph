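# Niagara skydiving contrail for the ZenCrash cosmetic: ropes with coins and tassels trailing from the player's skeletal mesh.
# Rope builds partitioned ribbons from bone positions, offset by curves and sine sway scaled by dive value.
# Coins and Tassel read Rope particles through scratch modules to stick and orient meshes along the ribbons, while WristRope pins two meshes to sockets.
# Paraglide lean and timeline values drive tassel rotation through dynamic material parameters.
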NiagaraSystem /BRCosmetics/Athena/Effects/Particles/Skydiving/NS_Contrail_ZenCrash/FX/NS_Contrail_ZenCrash
dependencies {
 usedDataInterfaces [
  &NiagaraDataInterfaceArrayFloat3
  &NiagaraDataInterfaceArrayInt32
  &NiagaraDataInterfaceColorCurve
  &NiagaraDataInterfaceArrayFloat
  &NiagaraDataInterfaceCurve
  &NiagaraDataInterfaceCamera
  &NiagaraDataInterfaceParticleRead
  &NiagaraDataInterfaceSkeletalMesh
 ]
 usedDynamicInputs [&SetBoolByFloatComparison, &Divide_Float]
 usedModules [
  &self:SystemSpawnScript.SystemScriptSource.SystemScriptGraph.NiagaraNodeAssignment_6.SetVariables_#1
  &self:SystemSpawnScript.SystemScriptSource.SystemScriptGraph.NiagaraNodeAssignment_8.SetVariables_#2
  &SystemState
  /Game/Effects/Niagara/Modules/Skydiving/FN_DivingAttributes/v6/FN_DivingAttributes
  /Game/Effects/Niagara/Modules/Contrails/FN_Contrail_Setup
  &self:ScratchModule
  &self:SystemSpawnScript.SystemScriptSource.SystemScriptGraph.NiagaraNodeAssignment_4.SetVariables_#3
  &self:SystemSpawnScript.SystemScriptSource.SystemScriptGraph.NiagaraNodeAssignment_1.SetVariables_#4
  &FN_Integrator
  &Timeline
 ]
}
emitters {
 Coins {
  emitterUpdateScript [
   &EmitterState(
    Inactive Response=::NewEnumerator1
    Life Cycle Mode=::NewEnumerator1
    Loop Behavior=::NewEnumerator1
    Loop Duration Mode=::NewEnumerator1
    UseLoopDelay=false
   )
   &SpawnBurst_Instantaneous(Spawn Count=&Multiply_Int(A=2, B=5), Use Spawn Probability=false, UseLoopCountLimit=false)
  ]
  particleSpawnScript [
   &FN_SetSpawnGroup(Number of Spawn Groups=2)
   &InitializeParticle(
    Color Mode=::NewEnumerator0
    Lifetime=5
    Mesh Scale=<2.5 2.5 2.5>
    Mesh Scale Mode=::NewEnumerator4
    Sprite Size Mode=::NewEnumerator3
    Uniform Sprite Size=100
    UsePositionOffset=false
    Write Material Random=true
   )
   &PartitionParticles(Number Of Partitions=$System.ChainNumber, Particles Per Partition=5, Partition Mode:ENiagara_PartitionMode=::NewEnumerator1)
   /SetVariables(Particles.NormalizedChainPosition:NiagaraFloat=$Particles.PartitionParticles.Normalized Index In Partition) {isSetParameters=true, module=SetVariables_#5}
   &SkeletalMeshLocation(
    Bone / Socket Index=$Particles.PartitionParticles.Partition
    Bone Sampling Mode=::NewEnumerator4
    MaskSpawnGroup=false
    Sampled Position Offset=&Lerp_Vector(A=<0 0 0>, B=<0 0 100>, Alpha=$Particles.NormalizedChainPosition)
    Sampled Position Offset Coordinate Space:ENiagaraSkelMeshTransforms=::NewEnumerator1
    Skeletal Mesh=$System.Skeletal Mesh
   )
   /SetVariables(Particles.MeshOrientation=&MultiplyQuaternion(
     Quaternion A=&MultiplyQuaternion(
      Quaternion A=&OrientMeshToVector(
       Look At Direction=&NormalizeVector(
        Vector To Normalize=&Multiply_VectorByFloat(
         Vector=<0 0 1>
         Float=&Lerp_Float(A=-1, B=1, Alpha=&MakeFloatFromInt(Int=$Particles.PartitionParticles.Partition))
        )
        Safe Normalize=true
        Safe Fallback Threshold=0
        Safe Fallback Vector=<0 0 0.001>
       )
       Reference Axis Vector=<1 0 0>
      )
      Quaternion B=&MultiplyQuaternion(
       Quaternion A=&MakeQuatFromAxisAngle(
        Axis=<0 1 0>
        AngleInDegrees=&Lerp_Float(A=90, B=-90, Alpha=&MakeFloatFromInt(Int=$Particles.PartitionParticles.Partition))
       )
       Quaternion B=&MakeQuatFromAxisAngle(
        Axis=<0 0 1>
        AngleInDegrees=&Lerp_Float(
         A=180
         B=&Lerp_Float(A=270, B=90, Alpha=&MakeFloatFromInt(Int=$Particles.PartitionParticles.Partition))
         Alpha=&FloatFromCurve(
          FloatCurve=DataInterface() {propertyValues {Curve {keys [
              {arriveTangent=2.7188, interpMode=RCIM_Cubic, leaveTangent=2.7188, tangentMode=RCTM_User, time=-1, value=-1}
              {interpMode=RCIM_Cubic, tangentMode=RCTM_User, time=0, value=0}
              {arriveTangent=2.2136, interpMode=RCIM_Cubic, leaveTangent=2.2136, tangentMode=RCTM_User, time=1, value=0.8}
             ]}}}
          CurveIndex=&DotProduct(
           Vector A=&NormalizeVector(
            Vector To Normalize=&Multiply_Vector(
             A=&Subtract_Position(
              A=$Particles.Position
              B=&ReturnCameraPosition(Coordinate Space:ENiagaraCoordinateSpace=::Simulation, Camera Query:NiagaraDataInterfaceCamera=DataInterface())
             )
             B=<1 1 1>
            )
            Safe Normalize=true
            Safe Fallback Threshold=0
            Safe Fallback Vector=<0 0 0.001>
           )
           Vector B=<0 0 1>
          )
         )
        )
        NormalizeAxis=true
       )
      )
     )
     Quaternion B=<0 0 0 1>
    )) {isSetParameters=true, module=SetVariables_#6}
  ]
  particleUpdateScript [
   &ParticleState(Kill Particles When Lifetime Has Elapsed=false)
   &ScaleMeshSize(Scale Factor=&VectorFromFloat(Value=&FloatFromCurve(
      FloatCurve=DataInterface() {propertyValues {Curve {defaultValue=1, keys [{tangentMode=RCTM_User, time=0, value=0}, {tangentMode=RCTM_User, time=0.5, value=1}]}}}
      CurveIndex=$Engine.System.Age
     )))
   &SkeletalMeshLocation(
    Bone / Socket Index=$Particles.PartitionParticles.Partition
    Bone Sampling Mode=::NewEnumerator4
    MaskSpawnGroup=false
    Position Sampling=::NewEnumerator2
    Sampled Position Offset=&Lerp_Vector(A=<20 -3 0>, B=<-20 3 1>, Alpha=&MakeFloatFromInt(Int=$Particles.PartitionParticles.Partition))
    Skeletal Mesh=$System.Skeletal Mesh
   ) {module=SkeletalMeshLocation001}
   &self:StickMeshToRibbon_03(
    Index:NiagaraInt32=$Particles.PartitionParticles.Index In Partition
    LinkOffset=7
    Particle Attribute Reader=DataInterface() {propertyValues {EmitterBinding {bindingMode=Other, emitterName=Rope}}}
    ParticlePerRibbon:NiagaraInt32=$System.ParticlePerRibbon
    Partition:NiagaraInt32=$Particles.PartitionParticles.Partition
    RibbonSpace=1
   ) {module=StickMeshToRibbon_03}
   &InterpolateOverTime(
    Rate Of Change:NiagaraFloat=&FN_FrontEndFloatModifier(Gameplay=1, Front End=10)
    Target Vector Value:Vector3f=&Multiply_VectorByFloat(
     Vector=&NormalizeVector(
      Vector To Normalize=&Multiply_Vector(
       A=&ReturnCameraVector(Camera Vectors=::NewEnumerator4, Coordinate Space=::Simulation, Origin Position=$Particles.Position)
       B=<1 1 1>
      )
      Safe Normalize=true
      Safe Fallback Threshold=0
      Safe Fallback Vector=<0 0 0.001>
     )
     Float=-1
    )
    Variable Type=::NewEnumerator2
   ) {enabled=false}
   &InterpolateOverTime(
    Rate Of Change=10
    Target Vector Value:Vector3f=&Multiply_VectorByFloat(Vector=$Particles.RibbonDirectionVector, Float=-1)
    Variable Type=::NewEnumerator2
   ) {module=InterpolateOverTime001}
   &self:OrientChainLinks_03(
    FrontVector:Vector3f=$Particles.InterpolateOverTime.InterpolateOverTime_V3_.Moving Average
    RightVector:Vector3f=<1 0 0>
    UpVector:Vector3f=$Particles.InterpolateOverTime001.InterpolateOverTime_V3_.Moving Average
   ) {module=OrientChainLinks_03}
   &SolveForcesAndVelocity(Clamp Velocity=false, Limit Acceleration=false)
  ]
  rendererClasses [&NiagaraMeshRendererProperties]
  renderers [{rendererClass=&NiagaraMeshRendererProperties, rendererIndex=0}]
 }
 Rope {
  emitterUpdateScript [
   &EmitterState(
    Inactive Response=::NewEnumerator1
    Life Cycle Mode=::NewEnumerator1
    Loop Behavior=::NewEnumerator1
    Loop Duration Mode=::NewEnumerator1
    UseLoopDelay=false
   )
   &SpawnBurst_Instantaneous(
    Spawn Count=&Multiply_Int(A=$System.ChainNumber, B=$System.ParticlePerRibbon)
    Use Spawn Probability=false
    UseLoopCountLimit=false
   )
  ]
  particleSpawnScript [
   &InitializeParticle(
    Color=&ColorFromCurve(
     ColorCurve=DataInterface() {propertyValues {
       AlphaCurve {keys [<at -0.0034 1 RCIM_Constant>,<at 0.0304 0 RCIM_Constant>,<at 0.108 1 RCIM_Constant>,<at 0.336 0 RCIM_Constant>,<at 0.4958 1 RCIM_Constant>,<at 0.5536 0 RCIM_Constant>,<at 0.6303 1 RCIM_Constant>,<at 0.8401 0 RCIM_Constant>]}
       BlueCurve {keys [<at 0 1 RCIM_Constant>,<at 1 1 RCIM_Constant>]}
       GreenCurve {keys [<at 0 1 RCIM_Constant>,<at 1 1 RCIM_Constant>]}
       RedCurve {keys [<at 0 1 RCIM_Constant>,<at 1 1 RCIM_Constant>]}
      }}
     CurveIndex=&ReturnNormalizedExecIndex()
    )
    Lifetime=5
    Mesh Scale Mode=::NewEnumerator4
    Ribbon Width=2.25
    Ribbon Width Mode=::NewEnumerator1
    Sprite Size Mode=::NewEnumerator3
    UsePositionOffset=false
    Write Material Random=true
   )
   &PartitionParticles(
    Number Of Partitions=$System.ChainNumber
    Particles Per Partition:NiagaraInt32=$System.ParticlePerRibbon
    Partition Mode:ENiagara_PartitionMode=::NewEnumerator1
   )
   /SetVariables(Particles.NormalizedChainPosition:NiagaraFloat=$Particles.PartitionParticles.Normalized Index In Partition) {isSetParameters=true, module=SetVariables_#5}
   /SetVariables() {isSetParameters=true}
  ]
  particleUpdateScript [
   &ParticleState(Kill Particles When Lifetime Has Elapsed=false)
   &SkeletalMeshLocation(
    Bone / Socket Index=$Particles.PartitionParticles.Partition
    Bone Sampling Mode=::NewEnumerator4
    MaskSpawnGroup=false
    Orientation Sampling=::NewEnumerator1
    Sampled Position Offset=&MakeVector(X=0, Y=-2, Z=0)
    Sampled Position Offset Coordinate Space:ENiagaraSkelMeshTransforms=::NewEnumerator2
    Skeletal Mesh=$System.Skeletal Mesh
   ) {module=SkeletalMeshLocation002}
   /SetVariables(Particles.Position=&AddVectorToPosition(
     Position=$Particles.Position
     Vector=&Multiply_VectorByFloat(
      Vector=&Multiply_VectorByFloat(
       Vector=&Add_Vector(
        A=&Multiply_Vector(
         A=&NormalizeVector(
          Vector To Normalize=&FN_FrontEndVector3Modifier(Gameplay=<0 0 0>, Front End=<0 0 0>)
          Safe Normalize=true
          Safe Fallback Threshold=0
          Safe Fallback Vector=<0 0 0.001>
         )
         B=&SelectVectorFromArray(
          Direct Array Index=$Particles.PartitionParticles.Partition
          Vector Selection Array=DataInterface() {propertyValues {FloatData [<-0.5 -0.5 -1>, <-0.5 -0.5 -1>, <0.1 -1.5 -1>, <0.1 -0.5 -1>, <-0.5 -1.5 -1>, <-0.5 -1 -1>, <0.1 -1 -1>, <0.1 -0.5 -1>]}}
         )
        )
        B=&Multiply_VectorByFloat(
         Vector=&TransformVector(
          Vector=&SelectVectorFromArray(
           Direct Array Index=$Particles.PartitionParticles.Partition
           Direct Array Mode=::NewEnumerator1
           Vector Selection Array=DataInterface() {propertyValues {FloatData [<2 -2 0>, <-2 0 0>, <1 2 0>, <-0.5 0 0>, <0.5 0 0>, <-1 2 0>, <1 1 0>, <-2 1 0>]}}
          )
          Destination Space=::Simulation
         )
         Float=&SelectFloatFromArray(
          Direct Array Index=&SpawnGroup()
          Float Selection Array=DataInterface() {propertyValues {FloatData [0.1,0.15,0.08,0.16,0.15,0.18]}}
         )
        )
       )
       Float=&FloatFromCurve(
        FloatCurve=DataInterface() {propertyValues {Curve {keys [{tangentMode=RCTM_User, time=0, value=0}, {tangentMode=RCTM_User, time=1, value=1}]}}}
        CurveIndex=$Particles.PartitionParticles.Normalized Index In Partition
        Scale Curve=&SelectFloatFromArray(
         Direct Array Index=&SpawnGroup()
         Float Selection Array=DataInterface() {propertyValues {FloatData [97,97,120,100,75,55,94,80]}}
        )
       )
      )
      Float=&FloatFromCurve(
       FloatCurve=DataInterface() {propertyValues {Curve {defaultValue=1, keys [{tangentMode=RCTM_User, time=0, value=0}, {tangentMode=RCTM_User, time=1, value=1}]}}}
       CurveIndex=$Engine.System.Age
      )
     )
    )) {isSetParameters=true}
   /SetVariables(Particles.Position=&AddVectorToPosition(
     Position=$Particles.Position
     Vector=&Multiply_VectorByFloat(
      Vector=/Niagara/DynamicInputs/Vectors/RotateVector(
       Vector To Rotate:Vector3f=&SelectVectorFromArray(
        Direct Array Index=$Particles.Initial.PartitionParticles.Partition
        Vector Selection Array=DataInterface() {propertyValues {FloatData [<-10 -5 3>, <15 5 1>]}}
       )
       Yaw:NiagaraFloat=&Add_Float(
        A=$Particles.Age
        B=&Multiply_Float(A=&MakeFloatFromInt(Int=&SpawnGroup()), B=2)
       )
       Pitch:NiagaraFloat=0
       Roll:NiagaraFloat=0
       Rotation Rate:NiagaraFloat=&SelectFloatFromArray(
        Direct Array Index=$Particles.PartitionParticles.Partition
        Float Selection Array=DataInterface() {propertyValues {FloatData [200,100,130,120,160,150,120,140]}}
       )
      )
      Float=&Multiply_Float(
       A=&Sine(
        Normalized Angle=&Add_Float(
         A=&Multiply_Float(
          A=$Emitter.Age
          B=&SelectFloatFromArray(
           Direct Array Index=&SpawnGroup()
           Float Selection Array=DataInterface() {propertyValues {FloatData [-2,-2.5,-2.5,-3,-3,-2,-2.5,-2]}}
          )
         )
         B=$Particles.PartitionParticles.Normalized Index In Partition
        )
        Period=0.9
        Scale=&Lerp_Float(A=0.8, B=1.4, Alpha=$System.FN_Contrail_Setup.Dive Value)
        Bias=0
       )
       B=&FloatFromCurve(
        FloatCurve=DataInterface() {propertyValues {Curve {keys [
            {arriveTangent=1.1676, interpMode=RCIM_Cubic, leaveTangent=1.1676, tangentMode=RCTM_User, time=0, value=0}
            {arriveTangent=1.4544, interpMode=RCIM_Cubic, leaveTangent=1.4544, tangentMode=RCTM_User, time=1, value=1}
           ]}}}
        CurveIndex=$Particles.PartitionParticles.Normalized Index In Partition
        Scale Curve=&FloatFromCurve(
         FloatCurve=DataInterface() {propertyValues {Curve {defaultValue=1, keys [{tangentMode=RCTM_User, time=0, value=0}, {tangentMode=RCTM_User, time=1, value=1}]}}}
         CurveIndex=$Engine.System.Age
        )
       )
      )
     )
    )) {isSetParameters=true}
   /SetVariables(Particles.Position=&AddVectorToPosition(
     Position=$Particles.Position
     Vector=&Multiply_VectorByFloat(
      Vector=&SelectVectorFromArray(
       Direct Array Index=$Particles.PartitionParticles.Partition
       Direct Array Mode=::NewEnumerator1
       Vector Selection Array=DataInterface() {propertyValues {FloatData [<-20 40 130>, <40 20 130>, <-25 0 0>, <25 0 0>]}}
      )
      Float=&FloatFromCurve(
       FloatCurve=DataInterface() {propertyValues {Curve {keys [{arriveTangent=0.9713, time=0, value=0}, <at 0.125 0.2>, <at 0.95 0.92>, {interpMode=RCIM_Cubic, leaveTangent=0.018, time=1, value=1}]}}}
       CurveIndex=$Particles.Initial.PartitionParticles.Normalized Index In Partition
       Scale Curve=&MaskFloat(
        Mask Source=1
        Mask Parameter=$Particles.PartitionParticles.Partition
        Mask=0
        Masked Float=&FloatFromCurve(
         FloatCurve=DataInterface() {propertyValues {Curve {
            defaultValue=1
            keys [
             {arriveTangent=0.4561, interpMode=RCIM_Cubic, leaveTangent=0.4561, tangentMode=RCTM_User, time=0, value=0}
             {arriveTangent=1, interpMode=RCIM_Cubic, leaveTangent=1, time=0.3297, value=0.3061}
             {arriveTangent=0.0854, interpMode=RCIM_Cubic, leaveTangent=0.0854, tangentMode=RCTM_User, time=1, value=1}
            ]
           }}}
         CurveIndex=$Engine.System.Age
        )
        Passthrough Float=&FloatFromCurve(
         FloatCurve=DataInterface() {propertyValues {Curve {
            defaultValue=1
            keys [
             {arriveTangent=-0.0467, interpMode=RCIM_Cubic, leaveTangent=-0.0467, tangentMode=RCTM_User, time=0, value=0}
             {arriveTangent=2.5859, interpMode=RCIM_Cubic, leaveTangent=2.5859, tangentMode=RCTM_User, time=0.196, value=0.1441}
             <at 1 1 RCIM_Cubic>
            ]
           }}}
         CurveIndex=$Engine.System.Age
        )
       )
      )
     )
    )) {isSetParameters=true}
   &ScaleRibbonWidth(Ribbon Width Scale=&FloatFromCurve(
     FloatCurve=DataInterface() {propertyValues {Curve {defaultValue=1, keys [{tangentMode=RCTM_User, time=0, value=0.25}, {tangentMode=RCTM_User, time=0.5, value=1}]}}}
     CurveIndex=$Engine.System.Age
     Scale Curve=0.75
    ))
   /SetVariables(
    Particles.Linear Color=&ColorFromCurve(
     ColorCurve=DataInterface() {propertyValues {
       AlphaCurve {keys [<at 0.0031 -0.0035 RCIM_Constant>,<at 0.0217 1 RCIM_Constant>,<at 0.1075 0 RCIM_Constant>,<at 0.5243 0.9965 RCIM_Constant>,<at 0.6314 0>]}
       BlueCurve {keys [<at 0 1 RCIM_Cubic>,<at 1 1 RCIM_Cubic>]}
       GreenCurve {keys [<at 0 1 RCIM_Cubic>,<at 1 1 RCIM_Cubic>]}
       RedCurve {keys [<at 0 1 RCIM_Cubic>,<at 1 1 RCIM_Cubic>]}
      }}
     CurveIndex=&ReturnNormalizedExecIndex()
    )
    Particles.float:NiagaraFloat=&FloatFromCurve(
     FloatCurve=DataInterface() {propertyValues {Curve {defaultValue=1, keys [{tangentMode=RCTM_User, time=0, value=0.25}, {tangentMode=RCTM_User, time=0.5, value=1}]}}}
     CurveIndex=$Engine.System.Age
     Scale Curve=&FloatFromCurve(
      FloatCurve=DataInterface() {propertyValues {Curve {keys [<at 0 1 RCIM_Constant>,<at 0.1914 0 RCIM_Constant>,<at 0.5011 1 RCIM_Constant>,<at 0.6267 0 RCIM_Constant>]}}}
      CurveIndex=&ReturnNormalizedExecIndex()
      Scale Curve=35
     )
    )
   ) {isSetParameters=true}
  ]
  rendererClasses [&NiagaraRibbonRendererProperties]
  renderers [
   {rendererClass=&NiagaraRibbonRendererProperties}
   {rendererClass=&NiagaraRibbonRendererProperties, rendererIndex=1}
  ]
 }
 Tassel {
  emitterUpdateScript [
   &EmitterState(
    Inactive Response=::NewEnumerator1
    Life Cycle Mode=::NewEnumerator1
    Loop Behavior=::NewEnumerator1
    Loop Duration Mode=::NewEnumerator1
    UseLoopDelay=false
   )
   &SpawnBurst_Instantaneous(Spawn Count=&Multiply_Int(A=$System.ChainNumber, B=1), Use Spawn Probability=false, UseLoopCountLimit=false)
  ]
  particleSpawnScript [
   &FN_SetSpawnGroup(Number of Spawn Groups=2)
   &InitializeParticle(
    Color=<rgba 0.5781 0 0 1>
    Lifetime=5
    Mesh Scale=<2.5 2.5 4>
    Mesh Scale Mode=::NewEnumerator4
    Sprite Size Mode=::NewEnumerator3
    Uniform Sprite Size=100
    UsePositionOffset=false
    Write Material Random=true
   )
   &PartitionParticles(Number Of Partitions=$System.ChainNumber, Particles Per Partition=2, Partition Mode:ENiagara_PartitionMode=::NewEnumerator1)
   /SetVariables(Particles.NormalizedChainPosition:NiagaraFloat=$Particles.PartitionParticles.Normalized Index In Partition) {isSetParameters=true, module=SetVariables_#5}
   &SkeletalMeshLocation(
    Bone / Socket Index=$Particles.PartitionParticles.Partition
    Bone Sampling Mode=::NewEnumerator4
    MaskSpawnGroup=false
    Sampled Position Offset=&Lerp_Vector(A=<0 0 0>, B=<0 0 100>, Alpha=$Particles.NormalizedChainPosition)
    Sampled Position Offset Coordinate Space:ENiagaraSkelMeshTransforms=::NewEnumerator1
    Skeletal Mesh=$System.Skeletal Mesh
   )
   /SetVariables(Particles.MeshOrientation=&MultiplyQuaternion(
     Quaternion A=&MultiplyQuaternion(
      Quaternion A=&OrientMeshToVector(
       Look At Direction=&NormalizeVector(
        Vector To Normalize=&Multiply_VectorByFloat(
         Vector=<0 0 1>
         Float=&Lerp_Float(A=-1, B=1, Alpha=&MakeFloatFromInt(Int=$Particles.PartitionParticles.Partition))
        )
        Safe Normalize=true
        Safe Fallback Threshold=0
        Safe Fallback Vector=<0 0 0.001>
       )
       Reference Axis Vector=<1 0 0>
      )
      Quaternion B=&MultiplyQuaternion(
       Quaternion A=&MakeQuatFromAxisAngle(
        Axis=<0 1 0>
        AngleInDegrees=&Lerp_Float(A=90, B=-90, Alpha=&MakeFloatFromInt(Int=$Particles.PartitionParticles.Partition))
       )
       Quaternion B=&MakeQuatFromAxisAngle(
        Axis=<0 0 1>
        AngleInDegrees=&Lerp_Float(
         A=180
         B=&Lerp_Float(A=270, B=90, Alpha=&MakeFloatFromInt(Int=$Particles.PartitionParticles.Partition))
         Alpha=&FloatFromCurve(
          FloatCurve=DataInterface() {propertyValues {Curve {keys [
              {arriveTangent=2.7188, interpMode=RCIM_Cubic, leaveTangent=2.7188, tangentMode=RCTM_User, time=-1, value=-1}
              {interpMode=RCIM_Cubic, tangentMode=RCTM_User, time=0, value=0}
              {arriveTangent=2.2136, interpMode=RCIM_Cubic, leaveTangent=2.2136, tangentMode=RCTM_User, time=1, value=0.8}
             ]}}}
          CurveIndex=&DotProduct(
           Vector A=&NormalizeVector(
            Vector To Normalize=&Multiply_Vector(
             A=&Subtract_Position(
              A=$Particles.Position
              B=&ReturnCameraPosition(Coordinate Space:ENiagaraCoordinateSpace=::Simulation, Camera Query:NiagaraDataInterfaceCamera=DataInterface())
             )
             B=<1 1 1>
            )
            Safe Normalize=true
            Safe Fallback Threshold=0
            Safe Fallback Vector=<0 0 0.001>
           )
           Vector B=<0 0 1>
          )
         )
        )
        NormalizeAxis=true
       )
      )
     )
     Quaternion B=<0 0 0 1>
    )) {isSetParameters=true, module=SetVariables_#6}
  ]
  particleUpdateScript [
   &ParticleState(Kill Particles When Lifetime Has Elapsed=false)
   &ScaleMeshSize(Scale Factor=&VectorFromFloat(Value=&FloatFromCurve(
      FloatCurve=DataInterface() {propertyValues {Curve {defaultValue=1, keys [{tangentMode=RCTM_User, time=0, value=0}, {tangentMode=RCTM_User, time=0.5, value=1}]}}}
      CurveIndex=$Engine.System.Age
     )))
   &SkeletalMeshLocation(
    Bone / Socket Index=$Particles.PartitionParticles.Partition
    Bone Sampling Mode=::NewEnumerator4
    MaskSpawnGroup=false
    Position Sampling=::NewEnumerator2
    Sampled Position Offset=&Lerp_Vector(A=<20 -3 0>, B=<-20 3 1>, Alpha=&MakeFloatFromInt(Int=$Particles.PartitionParticles.Partition))
    Skeletal Mesh=$System.Skeletal Mesh
   ) {module=SkeletalMeshLocation001}
   &self:StickMeshToRibbon_03(
    Index:NiagaraInt32=$Particles.PartitionParticles.Index In Partition
    LinkOffset=13
    Particle Attribute Reader=DataInterface() {propertyValues {EmitterBinding {bindingMode=Other, emitterName=Rope}}}
    ParticlePerRibbon=1
    Partition:NiagaraInt32=$Particles.PartitionParticles.Partition
    RibbonSpace=21
   ) {module=StickMeshToRibbon_03}
   &InterpolateOverTime(
    Rate Of Change:NiagaraFloat=&FN_FrontEndFloatModifier(Gameplay=1, Front End=10)
    Target Vector Value:Vector3f=&Multiply_VectorByFloat(
     Vector=&NormalizeVector(
      Vector To Normalize=&Multiply_Vector(
       A=&ReturnCameraVector(Camera Vectors=::NewEnumerator4, Coordinate Space=::Simulation, Origin Position=$Particles.Position)
       B=<1 1 1>
      )
      Safe Normalize=true
      Safe Fallback Threshold=0
      Safe Fallback Vector=<0 0 0.001>
     )
     Float=-1
    )
    Variable Type=::NewEnumerator2
   )
   &InterpolateOverTime(
    Rate Of Change=10
    Target Vector Value:Vector3f=&Multiply_VectorByFloat(Vector=$Particles.RibbonDirectionVector, Float=-1)
    Variable Type=::NewEnumerator2
   ) {module=InterpolateOverTime001}
   &self:OrientChainLinks_03(
    FrontVector:Vector3f=$Particles.InterpolateOverTime.InterpolateOverTime_V3_.Moving Average
    RightVector:Vector3f=<1 0 0>
    UpVector:Vector3f=$Particles.InterpolateOverTime001.InterpolateOverTime_V3_.Moving Average
   ) {module=OrientChainLinks_03}
   &SolveForcesAndVelocity(Clamp Velocity=false, Limit Acceleration=false)
   /SetVariables(Particles.Balloon_Orientation:Quat4f=&MultiplyQuaternion(
     Quaternion A=&MakeQuatFromAxisAngle(
      Axis=<0 0 1>
      AngleInDegrees=&SelectFloatFromArray(Direct Array Index=$Particles.PartitionParticles.Partition, Float Selection Array=DataInterface() {propertyValues {FloatData [5,-10]}})
     )
     Quaternion B=&MakeQuatFromAxisAngle(
      Axis=<0 0 -1>
      AngleInDegrees=&Lerp_Float(
       A=0
       B=&SelectFloatFromArray(Direct Array Index=$Particles.PartitionParticles.Partition, Float Selection Array=DataInterface() {propertyValues {FloatData [-50,50]}})
       Alpha=$User.ParaGlideLeanAlpha
      )
      NormalizeAxis=true
     )
    )) {isSetParameters=true}
   /SetVariables(Particles.Balloon_Orientation:Quat4f=&MultiplyQuaternion(
     Quaternion A=$Particles.Balloon_Orientation
     Quaternion B=&MakeQuatFromAxisAngle(
      Axis=&SelectVectorFromArray(
       Direct Array Index=$Particles.PartitionParticles.Partition
       Vector Selection Array=DataInterface() {propertyValues {FloatData [<-1 1 0>, <1 1 0>]}}
      )
      AngleInDegrees=&Lerp_Float(
       A=0
       B=&MakeCustomFloatFromBool(
        Bool=$User.bIsFrontEndPreview
        True Float=&SelectFloatFromArray(Direct Array Index=$Particles.PartitionParticles.Partition, Float Selection Array=DataInterface() {propertyValues {FloatData [-50,50]}})
        False Float=0
       )
       Alpha=$User.ParaGlideLeanAlpha
      )
      NormalizeAxis=true
     )
    )) {isSetParameters=true}
   /SetVariables(
    Particles.MaterialRotate1Axis:Vector3f=&TransformVector(
     Vector=&RotateVectorByQuaternion(
      Vector To Rotate=&Lerp_Vector(
       A=&SelectVectorFromArray(
        Direct Array Index=$Particles.PartitionParticles.Partition
        Vector Selection Array=DataInterface() {propertyValues {FloatData [<0.4 0.15 1>, <0.75 0.3 2>]}}
       )
       B=&SelectVectorFromArray(
        Direct Array Index=$Particles.PartitionParticles.Partition
        Vector Selection Array=DataInterface() {propertyValues {FloatData [<1 1 0>, <1 1 0>]}}
       )
       Alpha=$System.Timeline.NormalizedTime
      )
      Quaternion=$Particles.Balloon_Orientation
     )
     Destination Space=::World
    )
    Particles.MaterialRotateAxis:Vector3f=&TransformVector(
     Vector=&RotateVectorByQuaternion(
      Vector To Rotate=&Lerp_Vector(
       A=&SelectVectorFromArray(
        Direct Array Index=$Particles.PartitionParticles.Partition
        Vector Selection Array=DataInterface() {propertyValues {FloatData [<0.75 0.35 0.5>, <0.85 0.35 0.5>]}}
       )
       B=&SelectVectorFromArray(
        Direct Array Index=$Particles.PartitionParticles.Partition
        Vector Selection Array=DataInterface() {propertyValues {FloatData [<1 0.5 0.75>, <1 0.5 0.75>]}}
       )
       Alpha=$System.Timeline.NormalizedTime
      )
      Quaternion=$Particles.Balloon_Orientation
     )
     Destination Space=::World
    )
   ) {isSetParameters=true}
   &FN_Timer(Increment Multiplier:NiagaraFloat=&Lerp_Float(A=1, B=2, Alpha=$User.ParaGlideLeanAlpha))
   &DynamicMaterialParameters(
    Index 0 Param 1=&MakeFloatFromVector(Vector=$Particles.MaterialRotateAxis, Channel=::NewEnumerator0)
    Index 0 Param 2=&MakeFloatFromVector(Vector=$Particles.MaterialRotateAxis, Channel=::NewEnumerator1)
    Index 0 Param 3=&MakeFloatFromVector(Vector=$Particles.MaterialRotateAxis, Channel=::NewEnumerator2)
    Index 0 Param 4=&RandomRangeFloat(Minimum=0.2, Maximum=0.5)
    Index 1 Param 1=&MakeFloatFromVector(Vector=$Particles.MaterialRotate1Axis, Channel=::NewEnumerator0)
    Index 1 Param 2=&MakeFloatFromVector(Vector=$Particles.MaterialRotate1Axis, Channel=::NewEnumerator1)
    Index 1 Param 3=&MakeFloatFromVector(Vector=$Particles.MaterialRotate1Axis, Channel=::NewEnumerator2)
    Index 1 Param 4=&Lerp_Float(A=2, B=2, Alpha=$User.ParaGlideLeanAlpha)
    Index 2 Param 1=$Particles.FN_Timer.Value
    Index 2 Param 2=&Lerp_Float(A=0.75, B=1.25, Alpha=$User.ParaGlideLeanAlpha)
    Index 2 Param 3=&Lerp_Float(A=0.5, B=1, Alpha=$User.ParaGlideLeanAlpha)
    Index 2 Param 4=&Lerp_Float(A=0.75, B=2, Alpha=$User.ParaGlideLeanAlpha)
    Index 3 Param 1=0.85
    Index 3 Param 2=1
    Index 3 Param 3=0.66
    Module.Index0Param0Write=true
    Module.Index0Param1Write=true
    Module.Index0Param2Write=true
    Module.Index0Param3Write=true
    Module.Index1Param0Write=true
    Module.Index1Param1Write=true
    Module.Index1Param2Write=true
    Module.Index1Param3Write=true
    Module.Index2Param0Write=true
    Module.Index2Param1Write=true
    Module.Index2Param2Write=true
    Module.Index2Param3Write=true
    Module.Index3Param0Write=true
    Module.Index3Param1Write=true
    Module.Index3Param2Write=true
    Module.Index3Param3Write=true
    Write Parameter Index 1=true
    Write Parameter Index 2=true
    Write Parameter Index 3=true
   )
  ]
  rendererClasses [&NiagaraMeshRendererProperties]
  renderers [{rendererClass=&NiagaraMeshRendererProperties, rendererIndex=0}]
 }
 WristRope {
  emitterUpdateScript [
   &EmitterState(Life Cycle Mode=::NewEnumerator1, Loop Behavior=::NewEnumerator1, Loop Duration Mode=::NewEnumerator1, UseLoopDelay=false)
   &SpawnBurst_Instantaneous(Spawn Count=2, Use Spawn Probability=false, UseLoopCountLimit=false)
  ]
  particleSpawnScript [
   &PartitionParticles()
   &InitializeParticle(
    Color=<rgba 1.1875 0.0005 0.0166 1>
    Lifetime=5
    Mesh Scale Mode=::NewEnumerator3
    Mesh Uniform Scale=1.33
    UsePositionOffset=false
    Write Material Random=true
   )
   &InitialMeshOrientation(
    Mesh Axis To Orient=<0 1 0>
    Mesh Orientation Mode=::NewEnumerator2
    Rotation=&SelectVectorFromArray(
     Direct Array Index=$Particles.PartitionParticles.Partition
     Vector Selection Array=DataInterface() {propertyValues {FloatData [<0 0.1982 0>, <0 -0.6486 0>]}}
    )
    Use Rotation Vector=true
   )
  ]
  particleUpdateScript [
   &ParticleState(Kill Particles When Lifetime Has Elapsed=false, Let Infinitely Lived Particles Die When Emitter Deactivates=true)
   &DynamicMaterialParameters(
    Index 0 Param 2=&Lerp_Float(A=0.5, B=1, Alpha=$System.Timeline.Float)
    Index 0 Param 3=4
    Index 1 Param 2=$Engine.RealTime
    Module.Index0Param0Write=false
    Module.Index0Param1Write=true
    Module.Index0Param2Write=true
    Module.Index0Param3Write=false
    Module.Index1Param0Write=true
    Module.Index1Param1Write=true
    Module.Index1Param2Write=true
    Module.Index1Param3Write=true
    Write Parameter Index 1=true
   )
   &SkeletalMeshLocation(
    Bone / Socket Index=&SelectIntFromArray(Direct Array Index=$Particles.PartitionParticles.Partition, Int Selection Array=DataInterface() {propertyValues {IntData [0,1]}})
    MaskSpawnGroup=false
    Mesh Sampling Type=::NewEnumerator5
    Sampled Orientation Coordinate Space=::Local
    Sampled Position Offset Coordinate Space:ENiagaraSkelMeshTransforms=::NewEnumerator2
    Skeletal Mesh=$System.Skeletal Mesh
    Socket Sampling Mode=::NewEnumerator5
   )
   /SetVariables(Particles.VisibilityTag=&SelectIntFromArray(Direct Array Index=$Particles.PartitionParticles.Partition, Int Selection Array=DataInterface() {propertyValues {IntData [0,1]}})) {isSetParameters=true}
  ]
  rendererClasses [&NiagaraMeshRendererProperties]
  renderers [
   {rendererClass=&NiagaraMeshRendererProperties, rendererIndex=0}
   {rendererClass=&NiagaraMeshRendererProperties, rendererIndex=1}
  ]
 }
}
summary {
 systemName=NS_Contrail_ZenCrash
 emitters [
  {emitterName=WristRope, rendererClasses [&NiagaraMeshRendererProperties]}
  {emitterName=Rope, rendererClasses [&NiagaraRibbonRendererProperties]}
  {emitterName=Coins, rendererClasses [&NiagaraMeshRendererProperties]}
  {emitterName=Tassel, rendererClasses [&NiagaraMeshRendererProperties]}
 ]
 userVariables [
  User.bIsFrontEnd {type=NiagaraBool}
  User.bIsFrontEndPreview {type=NiagaraBool}
  User.ParaGlideLeanAlpha {type=NiagaraFloat, defaultValue=0, description=""}
  User.PlayerSpeed {type=NiagaraFloat, defaultValue=0, description=""}
  User.RibbonIDOffset {type=NiagaraInt32}
 ]
}
